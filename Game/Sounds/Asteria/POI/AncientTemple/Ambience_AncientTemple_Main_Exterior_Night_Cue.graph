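# Looping nighttime exterior rainforest ambience cue for the Asteria Ancient Temple POI, using outdoor ambient class and attenuation settings.
# A quality-level root routes its first two slots to a quad loop and its last two to a stereo loop.
# Each loop passes through an identical, endlessly repeating Enveloper that slowly varies pitch over a 40-second cycle while holding volume flat.
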SoundCue /Game/Sounds/Asteria/POI/AncientTemple/Ambience_AncientTemple_Main_Exterior_Night_Cue
root=QualityLevel_1
nodes [
 {
  id=Enveloper_0
  children [WavePlayer_0]
  properties {
   bLoop=true
   bLoopIndefinitely=true
   loopEnd=40
   pitchCurve {editorCurveData {keys [
      {arriveTangent=0, interpMode=RCIM_Cubic, leaveTangent=0, tangentMode=RCTM_User, time=0}
      {arriveTangent=0, interpMode=RCIM_Cubic, leaveTangent=0, tangentMode=RCTM_User, time=20, value=0.92}
      {arriveTangent=0.004, interpMode=RCIM_Cubic, leaveTangent=0.004, time=40}
     ]}}
   volumeCurve {editorCurveData {keys [<at 0 1>]}}
  }
 }
 {
  id=Enveloper_4
  children [WavePlayer_1]
  properties {
   bLoop=true
   bLoopIndefinitely=true
   loopEnd=40
   pitchCurve {editorCurveData {keys [
      {arriveTangent=0, interpMode=RCIM_Cubic, leaveTangent=0, tangentMode=RCTM_User, time=0}
      {arriveTangent=0, interpMode=RCIM_Cubic, leaveTangent=0, tangentMode=RCTM_User, time=20, value=0.92}
      {arriveTangent=0.004, interpMode=RCIM_Cubic, leaveTangent=0.004, time=40}
     ]}}
   volumeCurve {editorCurveData {keys [<at 0 1>]}}
  }
 }
 {id=QualityLevel_1, children [Enveloper_0,Enveloper_0,Enveloper_4,Enveloper_4]}
 {
  id=WavePlayer_0
  properties {bLooping=true, soundWaveAssetPtr=&/Ambience_Rainforest_AncientTemple_Night_Ext_Loop_Quad}
 }
 {
  id=WavePlayer_1
  properties {bLooping=true, soundWaveAssetPtr=&/Ambience_Rainforest_AncientTemple_Night_Ext_Loop_Stereo}
 }
]
properties {
 attenuationSettings=&Fort_Ambient_Athena_Att
 maxDistance=2097152
 soundClassObject=&Fort_Ambient_Outdoors_Athena
 volumeMultiplier=0.376
}
